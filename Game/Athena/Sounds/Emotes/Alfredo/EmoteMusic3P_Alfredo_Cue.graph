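# Third-person looping sound cue for the Alfredo emote, combining movement-reactive loops with emote music.
# EmoteMoveSpeed crossfades the idle loop toward the forward loop and raises idle pitch, under a half-second fade-in envelope.
# A licensed-content switcher, mixed at half volume, selects between silence and the music loop.
# The cue uses emote attenuation and concurrency, plays when silent, and feeds the music audio-analysis submix.
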SoundCue /Game/Athena/Sounds/Emotes/Alfredo/EmoteMusic3P_Alfredo_Cue
root=Mixer_1
nodes [
 {
  id=Enveloper_0
  children [Mixer_0]
  properties {pitchCurve {editorCurveData {keys [{time=0}]}}, volumeCurve {editorCurveData {keys [<at 0 0>,<at 0.5 1>]}}}
 }
 {id=FortSoundNodeLicensedContentSwitcher_0, class=FortSoundNodeLicensedContentSwitcher, children [WavePlayer_0,WavePlayer_0,WavePlayer_1]}
 {id=Mixer_0, children [ModulatorContinuous_1, ModulatorContinuous_0], properties {inputVolume [0.7,1]}}
 {id=Mixer_1, children [Enveloper_0, FortSoundNodeLicensedContentSwitcher_0], properties {inputVolume [1,0.5]}}
 {
  id=ModulatorContinuous_0
  children [WavePlayer_4]
  properties {pitchModulationParams {minOutput=1, parameterName=EmoteMoveSpeed}, volumeModulationParams {parameterName=EmoteMoveSpeed}}
 }
 {
  id=ModulatorContinuous_1
  children [WavePlayer_2]
  properties {
   pitchModulationParams {maxOutput=1.2, minOutput=1, parameterName=EmoteMoveSpeed}
   volumeModulationParams {maxOutput=0.2, minOutput=1, parameterName=EmoteMoveSpeed}
  }
 }
 {id=WavePlayer_0, properties {bLooping=true, soundWaveAssetPtr=/Game/Sounds/Fort_Test_Sounds/5_seconds_silence}}
 {id=WavePlayer_1, properties {bLooping=true, soundWaveAssetPtr=&/Emote_Alfredo_MX_LP}}
 {id=WavePlayer_2, properties {bLooping=true, soundWaveAssetPtr=&/Emote_Alfredo_Idle_LP}}
 {id=WavePlayer_4, properties {bLooping=true, soundWaveAssetPtr=&/Emote_Alfredo_Forward_LP}}
]
properties {
 attenuationSettings=&Fort_Emotes_Athena_Att
 maxDistance=1600
 soundClassObject=/Game/Sounds/FortSoundClassesAndModes/Fort_Music_Emote_3P
 virtualizationMode=PlayWhenSilent
 volumeMultiplier=1
 concurrencySet [/Game/Sounds/Concurrency_Rules/Athena_EmoteMusic_Conc]
 soundSubmixSends [{minSendLevel=0, soundSubmix=&Fortnite_AudioAnalysis_Music}]
}
